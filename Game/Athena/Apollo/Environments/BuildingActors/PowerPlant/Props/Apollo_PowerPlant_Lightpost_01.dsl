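; Blueprint for a power plant lightpost building prop in the Apollo environment, with a spot light component.
; Its construction script runs the parent construction logic, then checks whether the platform is Nintendo Switch.
; On Switch it turns off shadow casting on the spot light.
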(fn ConstructionScript
  (|CallParentFunction)
  (if (TouchPlatform|IsNintendoSwitch)
    (Rendering|Components|Light|SetCastShadows (Variables|Default|GetSpotLight) false)))
